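# Niagara effect for the Badger_Bangs (BlackMonday) prototype weapon that, per its name, shows an orange-red glow on the in-hand projectile.
# One active emitter burst-spawns a looping particle with two sprite renderers and a light, placed at an offset skeletal mesh bone sample.
# Emitter-age curves drive an opening flare that quickly shrinks and dims, with a secondary sprite size collapsing to zero.
# The three CoreFlare emitters and the rate spawner are disabled.
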NiagaraSystem /FNGameplayCues/Athena/Items/Weapons/Prototype/Badger_Bangs/FX/NS_BlackMonday_PrjOnHand
dependencies {
 usedDataInterfaces [&NiagaraDataInterfaceCurve, &NiagaraDataInterfaceSkeletalMesh]
 usedDynamicInputs []
 usedModules [
  &self:SystemSpawnScript.SystemScriptSource.SystemScriptGraph.NiagaraNodeAssignment_0.SetVariables_#1
  &SystemState
  &SampleSkeletalMesh
 ]
}
emitters {
 CoreFlare {bEnabled=false, rendererClasses [], renderers []}
 CoreFlare1 {bEnabled=false, rendererClasses [], renderers []}
 CoreFlareStatic {bEnabled=false, rendererClasses [], renderers []}
 SingleLoopingParticle {
  emitterUpdateScript [
   &EmitterState(
    Inactive Response=::NewEnumerator1
    Life Cycle Mode=::NewEnumerator1
    Loop Behavior=::NewEnumerator1
    Loop Duration Mode=::NewEnumerator1
    UseLoopDelay=false
   )
   &SpawnBurst_Instantaneous(Use Spawn Probability=false, UseLoopCountLimit=true)
   &SpawnRate(SpawnGroup=1, SpawnRate=3, Use Spawn Probability=false) {enabled=false}
  ]
  particleSpawnScript [
   &InitializeParticle(
    Color=<rgba 1 0.0762 0 1>
    Lifetime=&MaskFloatBySpawnGroup(Masked Float=$Emitter.CurrentLoopDuration, Passthrough Float=0.15)
    Sprite Size=<75 35>
    Sprite Size Mode=::NewEnumerator4
    UsePositionOffset=false
    Write Material Random=true
   )
   /SetVariables(Particles.VisibilityTag=&SpawnGroup()) {isSetParameters=true}
   &CameraOffset(Camera Offset Amount=40)
  ]
  particleUpdateScript [
   &ParticleState()
   &ScaleSpriteSize(
    Uniform Curve Index=$Emitter.Age
    Uniform Curve Sprite Scale=DataInterface() {propertyValues {Curve {keys [{tangentMode=RCTM_User, time=0, value=8}, {tangentMode=RCTM_User, time=0.15, value=1}]}}}
   )
   &SolveForcesAndVelocity(Clamp Velocity=false, Limit Acceleration=false)
   &ScaleColor(
    Scale RGB=&VectorFromFloat(Value=&FloatFromCurve(FloatCurve=DataInterface() {propertyValues {Curve {keys [
          {arriveTangent=9.0725, interpMode=RCIM_Cubic, leaveTangent=9.0725, leaveTangentWeight=0.149, tangentMode=RCTM_User, time=0, value=0}
          {
           arriveTangent=-7.5557
           arriveTangentWeight=0.0171
           interpMode=RCIM_Cubic
           leaveTangent=-7.5557
           leaveTangentWeight=0.1319
           tangentMode=RCTM_User
           time=0.0513
           value=0.9
          }
          {
           arriveTangent=-0.0405
           arriveTangentWeight=0.1003
           interpMode=RCIM_Cubic
           leaveTangent=-0.0405
           leaveTangentWeight=0.0316
           tangentMode=RCTM_User
           time=0.276
           value=0.1865
          }
          {arriveTangent=1.7265, interpMode=RCIM_Cubic, leaveTangent=1.7265, leaveTangentWeight=0.1559, tangentMode=RCTM_User, time=0.4508, value=0.287}
          {
           arriveTangent=-7.5557
           arriveTangentWeight=0.1302
           interpMode=RCIM_Cubic
           leaveTangent=-7.5557
           leaveTangentWeight=0.7642
           tangentMode=RCTM_User
           time=0.4982
           value=0.837
          }
          {
           arriveTangent=-0.0405
           arriveTangentWeight=0.1003
           interpMode=RCIM_Cubic
           leaveTangent=-0.0405
           tangentMode=RCTM_User
           time=0.799
           value=-0.063
          }
         ]}}}))
    ScaleA=true
    ScaleRGB=true
   ) {enabled=false}
   /SetVariables(
    Particles.Color_B:LinearColor=&Multiply_LinearColorByFloat(Linear Color=$Particles.Color, Float=0.4)
    Particles.SpriteSize_B:Vector2f=&Multiply_Vector2D_ByFloat(
     Float=&FloatFromCurve(
      FloatCurve=DataInterface() {propertyValues {Curve {keys [
          {arriveTangent=-3.4637, interpMode=RCIM_Cubic, leaveTangent=-3.4637, tangentMode=RCTM_User, time=0, value=1}
          {arriveTangent=-5, interpMode=RCIM_Cubic, leaveTangent=-5, time=0.0739, value=5.1153}
          {arriveTangent=-0.0499, interpMode=RCIM_Cubic, leaveTangent=-0.0499, tangentMode=RCTM_User, time=0.2, value=0}
         ]}}}
      CurveIndex=$Emitter.Age
     )
     Vector2D=<30 30>
    )
   ) {isSetParameters=true}
   /SetVariables(Particles.Color_C:LinearColor=&LerpLinearColors(StartColor=$Particles.Color, EndColor=$Particles.Color_B, LerpFactor=1)) {isSetParameters=true}
   &ScaleColor(
    Scale RGB=&VectorFromFloat(Value=&FloatFromCurve(FloatCurve=DataInterface() {propertyValues {Curve {keys [<at 0 5>,<at 0.15 1>]}}}, CurveIndex=$Emitter.Age))
    ScaleA=true
    ScaleRGB=true
   ) {module=ScaleColor001}
   &SkeletalMeshLocation(
    Bone Sampling Mode=::NewEnumerator4
    MaskSpawnGroup=false
    Sampled Position Offset=<-10 -10 5>
    Skeletal Mesh=$System.Skeletal Mesh
    Use Sampled Mesh Attribute as Ribbon ID=false
   )
  ]
  rendererClasses [&NiagaraSpriteRendererProperties, &NiagaraLightRendererProperties]
  renderers [
   {rendererClass=&NiagaraSpriteRendererProperties}
   {rendererClass=&NiagaraSpriteRendererProperties, rendererIndex=1}
   {rendererClass=&NiagaraLightRendererProperties, rendererIndex=2}
  ]
 }
}
summary {
 systemName=NS_BlackMonday_PrjOnHand
 emitters [
  {
   emitterName=SingleLoopingParticle
   rendererClasses [&NiagaraSpriteRendererProperties, &NiagaraLightRendererProperties]
  }
  {bEnabled=false, emitterName=CoreFlare, rendererClasses []}
  {bEnabled=false, emitterName=CoreFlareStatic, rendererClasses []}
  {bEnabled=false, emitterName=CoreFlare1, rendererClasses []}
 ]
 userVariables []
}
